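# M_Cube_Cosmetics is a masked, two-sided Default Lit material for the SpacePlunge backpack cosmetic, usable on skeletal and instanced static meshes.
# It inserts MF_SpacePlungeCube between the 2023 CharacterParent PreFX and PostFX functions, passing material attributes through the chain.
# The cube effect's mask is a constant one.
# The PreFX M output also feeds PostFX's M, pre-skinned position, distance blend and Fresnel inputs.
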Material /BRCosmetics/Accessories/FORT_Backpacks/Backpack_M_MED_SpacePlunge/Material/M_Cube_Cosmetics
graph {
 properties {
  bUseMaterialAttributes=true
  bUsedWithInstancedStaticMeshes=true
  bUsedWithMeshDeformer=true
  bUsedWithSkeletalMesh=true
  blendMode=BLEND_Masked
  pixelDepthOffsetMode=PDOM_Legacy
  twoSided=true
  usedShadingModels=Default Lit
 }
 nodes [
  n1 = MaterialFunctionCall() {materialFunction=&MF_CharacterParent_2023_PreFX}
  MaterialFunctionCall(
   Material=MaterialFunctionCall!(In=@n1.Result, Mask=Constant() {r=1}) {materialFunction=&/MF_SpacePlungeCube}.Result
   M=@n1.M
   Pre Skinned Local Position=@n1.M
   Distance Blend=@n1.M
   Fresnel=@n1.M
  ) {materialFunction=&MF_CharacterParent_2023_PostFX}
 ]
}
