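# Animation Blueprint for the SpacePlunge backpack, with additive bone offsets on dyn_cube and pack.
# Each update computes an Ambient Translation from layered sine waves of real time for a floating drift.
# On customization load, a wildcard match on a character part's name flags whether the wearer lacks the SpacePlunge skin.
# After evaluation, it writes the CubeFX socket location to the materials' CenterPosition parameter.
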AnimBlueprint /BRCosmetics/Accessories/FORT_Backpacks/Backpack_M_MED_SpacePlunge/Meshes/M_MED_SpacePlunge_Pack_AnimBP
target_skeleton=&/M_MED_SpacePlunge_Pack_Skeleton
graphs [
 AnimGraph {nodes [
   n1 = "Transform (Modify) Bone - Bone: dyn_cube"!(ComponentPose=Local To Component(LocalPose=Input Pose()), bAlphaBoolEnabled:bool=True, Translation="0,0,0") -> (Pose) {
    alphaInputType=Bool
    rotationSpace=BCS_WorldSpace
    translationMode=BMM_Additive
    translationSpace=BCS_BoneSpace
    alphaBoolBlend {blendInTime=4, blendOption=Cubic, blendOutTime=4}
    boneToModify {boneName=dyn_cube}
   }
   n2 = "Transform (Modify) Bone - Bone: pack"!(ComponentPose=@n1, bAlphaBoolEnabled:bool, Translation="0,2.5,0") -> (Pose) {alphaInputType=Bool, translationMode=BMM_Additive, translationSpace=BCS_BoneSpace, boneToModify {boneName=pack}, translation {y=2.5}}
   Output Pose(Result=Component To Local(ComponentPose=@n2)) {layerGroup=None}
  ]}
 EventGraph {nodes [
   Event BlueprintUpdateAnimation()
   n1 = GetRealTimeSeconds()
   n2 = float * float(A="Sin (Radians)"(A=float * float(A=1.2, B=@n1)), B=0.15)
   n3 = float * float(A="Sin (Radians)"(A=float * float(A=1, B=@n1)), B=0.3)
   n4 = Remap Float(Input="Sin (Radians)"(A=float * float(A=1.4, B=@n1)), OldRange=(X=-1,Y=1), NewRange=(X=-1,Y=2)) -> (Output=~)
   n5 = float * float(A="Clamp (Float)"(Value=@n4, Min=-1, Max=0), B=0.15)
   > Set Ambient Translation(Ambient Translation:Vector=vector + vector(A=MakeVector(X=@n2, Y=@n3, Z=@n4), B=MakeVector(Y=@n5))) -> (Output_Get:Vector)
   n6 = Event OnCharacterCustomizationLoaded()
   > Set Skel_Mesh(Skel_Mesh=Get Array Item with EFort Custom Part Type(Array=@n6.CustomMeshParts))
   n7 = GetDisplayName(Object=Get Skel_Mesh())
   > Set isNotCustomSkin(isNotCustomSkin:bool=NOT Boolean(A=MatchesWildcard(SourceString=@n7, Wildcard=*SpacePlunge*))) -> (Output_Get:bool)
   Event BlueprintPostEvaluateAnimation() {is=title K2Node_Event}
   n8 = GetPartSkeletalMeshComponent()
   > Is Valid(InputObject=@n8)
   > SetVectorParameterValueOnMaterials!(
    execute
    self:Mesh Component Object Reference=@n8
    ParameterName:Name=CenterPosition
    ParameterValue:Vector=GetSocketLocation!(self:Scene Component Object Reference=@n8, InSocketName:Name=CubeFX) -> (ReturnValue:Vector)
   ) -> (then)
  ]}
]
variables [
 Ambient Translation {type=Vector}
 isNotCustomSkin {type=boolean}
 Skel_Mesh {type=SkeletalMeshComponent}
]
