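# Material function for the SpacePlunge backpack that outputs animated, glitched hologram-style scanlines plus a SpaceNoise mask.
# UVs blend between stabilized screen-space and pre-skinned position XZ, then pass through a screen-glitch function.
# The atlas scanline result is multiplied by a sine pulse with random flicker flashes from panned digital noise, then accent secondary lines are added.
# SpaceNoise sine-cycles a three-way lerp across that panned noise's RGB channels.
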MaterialFunction /BRCosmetics/Accessories/FORT_Backpacks/Backpack_M_MED_SpacePlunge/Material/MF_Scanlines
graph {nodes [
  n1 = ScalarParameter() {defaultValue=0.004, parameterName=ScanlineSpeed}
  n2 = FunctionInput(Preview=VertexInterpolator(VS=PreSkinnedPosition()).PS) {bUsePreviewValueAsDefault=true, inputName=PreskinnedPOS}
  n3 = LinearInterpolate(
   A=MaterialFunctionCall!(
    ObjectPivotLocation=VectorParameter() {group=Stablized, parameterName=CenterPosition, defaultValue=<rgba 0 0 0 1>}.RGB
    ObjectPositionOffset=Constant3Vector()
   ) {materialFunction=/Game/Characters/Player/Common/MaterialFunctions/Superlevel/ImprovedStabilizedScreenSpaceUVs}.StabilizedUVs
   B=Divide(A=ComponentMask(None=@n2) {b=true, r=true}, B=ScalarParameter() {defaultValue=100, parameterName=ScanLineUnit_Preskinned})
   Alpha=ScalarParameter() {group=Stablized, parameterName=UsePreskinned}
  )
  n4 = Time()
  n5 = MaterialFunctionCall!(
   Number of Scanlines=FunctionInput(Preview=ScalarParameter() {defaultValue=50, parameterName=Scanline amount}) {bUsePreviewValueAsDefault=true, inputName=Scanline Amount, inputType=FunctionInput_Scalar}
   Scanline Speed=@n1
   "Scanlines Thickness (Bias)"
   Scanlines Dark Value=ScalarParameter() {defaultValue=0.5, parameterName=Scanline Darkness}
   Scanlines Light Value
   Block Opacity=ScalarParameter() {defaultValue=1, parameterName=BlockOpacity}
   UVs=Multiply(
    A=MaterialFunctionCall!(Main_TextureUV=@n3) {materialFunction=&/MF_ScreenGlitch_SpacePlunge}.Result
    B=ScalarParameter() {defaultValue=1, parameterName=NoiseTiling}
   )
   Time=Multiply(A=Multiply(A=@n4, B=@n1)) {constB=20}
   Block Repeat=Constant() {r=3}
  ) {materialFunction=/Game/Effects/Fort_Effects/Effects/Atlas/Materials/MF_Atlas_Scanlines_01}
  n6 = ScalarParameter() {defaultValue=4, parameterName=VerticalNoiseTiling01}
  n7 = Panner(Coordinate=MaterialFunctionCall(Main_TextureUV=Multiply(A=@n3, B=@n6)) {materialFunction=&MF_ScreenGlitch_SpacePlunge}.Result) {speedY=0.2}
  n8 = TextureSample(UVs=@n7) {samplerType=SAMPLERTYPE_LinearColor, texture=/Game/Global/Textures/Default/Noise/T_Atlas_DigitalNoise_01}
  n9 = Add(A=@n8.R, B=Multiply(A=Time(), B=ScalarParameter() {defaultValue=0.2, parameterName=LineSineSpeed01}))
  n10 = SmoothStep(Value=MaterialFunctionCall() {materialFunction=/Game/MF__FloatFlickerRandom_Generic}.RandomFlickerFloat) {constMin=0.95}
  n11 = Add(
   A=ConstantBiasScale(None=Sine(None=@n9))
   B=Multiply(A=Multiply(A=@n8.G, B=@n10), B=ScalarParameter() {defaultValue=1, parameterName=FlashIntensity})
  )
  n12 = ScalarParameter() {defaultValue=1, parameterName=Accent Scanlight Brightness}
  FunctionOutput(None=Add(A=Multiply(A=@n5.Result, B=@n11), B=Multiply(A=@n5.Secondary Lines, B=@n12) {constB=3}))
  n13 = TextureSample(UVs=@n7) {samplerType=SAMPLERTYPE_LinearColor, texture=&T_Atlas_DigitalNoise_01}
  n14 = Sine(None=Time())
  FunctionOutput(None=MaterialFunctionCall(
    A=ComponentMask(None=@n13.RGB) {r=true}
    B=ComponentMask(None=@n13.RGB) {g=true}
    C=ComponentMask(None=@n13.RGB) {b=true}
    Alpha=ConstantBiasScale(None=@n14)
   ) {materialFunction=&Lerp_3Color}.Result) {outputName=SpaceNoise, sortPriority=1}
  MaterialFunctionCall(Main_TextureUV=Multiply(A=@n3, B=ScalarParameter() {defaultValue=3, parameterName=VerticalNoiseTiling})) {materialFunction=&MF_ScreenGlitch_SpacePlunge}
 ]}
